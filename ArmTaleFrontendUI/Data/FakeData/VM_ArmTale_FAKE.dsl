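; Fake-data MVVM viewmodel for the ArmTale frontend UI, supplying weapon and mod lists and equipping mods to weapon slots.
; Weapon and mod entry viewmodels are constructed from placeholder DataTables.
; The mod list, built only when the slot's equipped-mod object is valid, prepends "None", filters by slot type and flags the equipped item.
; Changing slot broadcasts a FieldNotify for that list; equipping clears a matching preview mod.
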(fn EquipModToSelectedWeapon
  (params (ModToEquip :type VM%20Arm%20Tale%20Mod%20Entry%20FAKE%20Object%20Reference))
  (Default|EquipModtoWeapon (Variables|Default|GetSelectedWeapon) (Variables|Default|GetSelectedModSlotIndex) ModToEquip)
  ; Clear the previewed mod if it's equipped
  (if (== (Variables|Default|GetPreviewModForSelectedWeapon) (Variables|GetModToEquip))
    (Variables|Default|SetPreviewModForSelectedWeapon 0)))

(fn FilteredModListEntries
  (bind _selectedweapon (Variables|Default|GetSelectedWeapon))
  (bind (_weapondata_itemdefinition_64_e664a9f348977afdcea70ba6c5829726 _weapondata_weaponlevelprogress_41_20c68c5e4244de8471b91bba848297e2 _weapondata_stat0_43_fc7115904a1bf033cc36d0a82fd3a8dc _weapondata_stat1_47_0a2f03a74b3f5506fa4de2a70ef45d5e _weapondata_stat2_48_b2178bd947257b986a0a74ad3b8c2fa1 _weapondata_stat3_49_2b1036a945a55bd6ec75ddb2e2f44d97 _weapondata_modslots_73_5397b13c42ca59a5b3c99aa71d2e1c9c) (Utilities|Struct|BreakSArmTaleWeaponDataFAKE (Variables|/ArmTaleFrontendUI/Data/FakeData/VM_ArmTale_WeaponEntry_FAKE.VM_ArmTale_WeaponEntry_FAKE_C|GetWeaponData_ItemDefinition_64_E664A9F348977AFDCEA70BA6C5829726 _selectedweapon)))
  (bind _selectedmodslotindex (Variables|Default|GetSelectedModSlotIndex))
  ; Cache the currently equipped mod so that we can set it on the constructed entry
  (bind _equippedmod (CallFunction|GetEquippedModforSelectedSlot))
  (Utilities|IsValid _equippedmod
    (:Is%20Valid
      (bind _moddata_itemdefinition_34_d7c19a1b47268b3103b0918b3510f954 (Utilities|Struct|BreakSArmTaleModDataFAKE (Variables|/ArmTaleFrontendUI/Data/FakeData/VM_ArmTale_ModEntry_FAKE.VM_ArmTale_ModEntry_FAKE_C|GetModData_ItemDefinition_34_D7C19A1B47268B3103B0918B3510F954 _equippedmod)))
      (Utilities|IsValid _moddata_itemdefinition_34_d7c19a1b47268b3103b0918b3510f954
        (:Is%20Valid
          (LocalVariables|SetCurrentlyEquippedMod _moddata_itemdefinition_34_d7c19a1b47268b3103b0918b3510f954))
        (:Is%20Not%20Valid
          _))
      ; Get the Current Mod Type Filter from the Selected Weapon
      (if (Utilities|Array|IsValidIndex _weapondata_modslots_73_5397b13c42ca59a5b3c99aa71d2e1c9c _selectedmodslotindex)
        (LocalVariables|SetCurrentModTypeFilter (Utilities|Array|Get(acopy) _weapondata_modslots_73_5397b13c42ca59a5b3c99aa71d2e1c9c _selectedmodslotindex))
        (else
          (LocalVariables|SetCurrentModTypeFilter "NewEnumerator3")))
      ; Make the first entry a "None" option
      (bind _returnvalue (Game|ConstructObjectfromClass "/ArmTaleFrontendUI/Data/FakeData/VM_ArmTale_ModEntry_FAKE.VM_ArmTale_ModEntry_FAKE_C"))
      (Utilities|IsValid (LocalVariables|GetCurrentlyEquippedMod)
        (:Is%20Valid
          _)
        (:Is%20Not%20Valid
          (Variables|/ArmTaleFrontendUI/Data/FakeData/VM_ArmTale_ModEntry_FAKE.VM_ArmTale_ModEntry_FAKE_C|SetIsEquipped true _returnvalue)))
      (Utilities|Array|Add (LocalVariables|GetOutListEntries) _returnvalue)
      (Variables|/ArmTaleFrontendUI/Data/FakeData/VM_ArmTale_ModEntry_FAKE.VM_ArmTale_ModEntry_FAKE_C|SetModData_ItemDefinition_34_D7C19A1B47268B3103B0918B3510F954 (Utilities|Struct|MakeSArmTaleModDataFAKE 0 (LocalVariables|GetCurrentModTypeFilter)) _returnvalue)
      ; Construct the filtered array from the Data Table
      (bind _outrownames (DataTable|GetDataTableRowNames "/ArmTaleFrontendUI/Data/FakeData/DT_ArmTale_Mods_FAKE.DT_ArmTale_Mods_FAKE"))
      (for _array_element _outrownames
        (bind _returnvalue_2 (Utilities|GetDataTableRow "/ArmTaleFrontendUI/Data/FakeData/DT_ArmTale_Mods_FAKE.DT_ArmTale_Mods_FAKE" _array_element))
        (if (or (Utilities|Enum|Equal(Enum) (LocalVariables|GetCurrentModTypeFilter) "NewEnumerator3") (Utilities|Enum|Equal(Enum) (LocalVariables|GetCurrentModTypeFilter) (.ModType_30_20C68C5E4244DE8471B91BBA848297E2 (Utilities|Struct|BreakSArmTaleModDataFAKE _returnvalue_2))))
          (bind _returnvalue_3 (Game|ConstructObjectfromClass "/ArmTaleFrontendUI/Data/FakeData/VM_ArmTale_ModEntry_FAKE.VM_ArmTale_ModEntry_FAKE_C"))
          (bind _output_get (Variables|/ArmTaleFrontendUI/Data/FakeData/VM_ArmTale_ModEntry_FAKE.VM_ArmTale_ModEntry_FAKE_C|SetModData _returnvalue_2 _returnvalue_3))
          (bind _output_get_itemdefinition_34_d7c19a1b47268b3103b0918b3510f954 (Utilities|Struct|BreakSArmTaleModDataFAKE _output_get))
          (Variables|/ArmTaleFrontendUI/Data/FakeData/VM_ArmTale_ModEntry_FAKE.VM_ArmTale_ModEntry_FAKE_C|SetIsEquipped (== _output_get_itemdefinition_34_d7c19a1b47268b3103b0918b3510f954 (LocalVariables|GetCurrentlyEquippedMod)) _returnvalue_3)
          (Utilities|Array|Add (LocalVariables|GetOutListEntries) _returnvalue_3)))
      (return (LocalVariables|GetOutListEntries)))
    (:Is%20Not%20Valid
      _)))

(fn GetEquippedModForSelectedSlot
  (bind _equippedmod (Class|VMArmTaleWeaponEntryFAKE|GetEquippedModforSlot (Variables|Default|GetSelectedWeapon) (Variables|Default|GetSelectedModSlotIndex)))
  (return _equippedmod))

(fn SetSelectedModSlotIndex
  (params (ModSlotIndex :type Integer))
  (Variables|Default|SetSelectedModSlotIndex ModSlotIndex)
  (FieldNotification|BroadcastFieldValueChanged self "(FieldName=\"FilteredModListEntries\")"))

(fn WeaponListEntries
  ; Construct the weapons array from the Data Table
  (bind _outrownames (DataTable|GetDataTableRowNames "/ArmTaleFrontendUI/Data/FakeData/DT_ArmTale_Weapons_FAKE.DT_ArmTale_Weapons_FAKE"))
  (for _array_element _outrownames
    (bind _returnvalue (Utilities|GetDataTableRow "/ArmTaleFrontendUI/Data/FakeData/DT_ArmTale_Weapons_FAKE.DT_ArmTale_Weapons_FAKE" _array_element))
    (bind _returnvalue_2 (Game|ConstructObjectfromClass "/ArmTaleFrontendUI/Data/FakeData/VM_ArmTale_WeaponEntry_FAKE.VM_ArmTale_WeaponEntry_FAKE_C"))
    (Variables|/ArmTaleFrontendUI/Data/FakeData/VM_ArmTale_WeaponEntry_FAKE.VM_ArmTale_WeaponEntry_FAKE_C|SetWeaponData _returnvalue _returnvalue_2)
    (Utilities|Array|Add (LocalVariables|GetOutListEntries) _returnvalue_2))
  (return (LocalVariables|GetOutListEntries)))
