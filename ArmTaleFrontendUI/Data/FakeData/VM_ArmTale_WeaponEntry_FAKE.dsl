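; Fake-data weapon entry viewmodel for the ArmTale frontend UI, tracking the mods equipped in each of a weapon's mod slots.
; EquippedMods rebuilds the list from the weapon data's mod slots, reusing cached entries and constructing typed placeholder mod viewmodels for uncached slots.
; EquipModToWeapon swaps a slot's cached mod, updates both IsEquipped flags and broadcasts the EquippedMods field change.
; GetEquippedModForSlot reads directly from that cache.
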(fn EquipModToWeapon
  (params (ModSlotIndex :type Integer) (ModToEquip :type VM%20Arm%20Tale%20Mod%20Entry%20FAKE%20Object%20Reference))
  (bind _modtoequip (Variables|GetModToEquip))
  (bind _modslotindex (Variables|GetModSlotIndex))
  (bind _cacheequippedmods (Variables|Default|GetCacheEquippedMods))
  (if (>= ModSlotIndex 0)
    (Variables|/ArmTaleFrontendUI/Data/FakeData/VM_ArmTale_ModEntry_FAKE.VM_ArmTale_ModEntry_FAKE_C|SetIsEquipped false (Utilities|Array|Get(acopy) _cacheequippedmods _modslotindex))
    (Utilities|Array|SetArrayElem _cacheequippedmods _modslotindex _modtoequip)
    (Variables|/ArmTaleFrontendUI/Data/FakeData/VM_ArmTale_ModEntry_FAKE.VM_ArmTale_ModEntry_FAKE_C|SetIsEquipped true _modtoequip)
    (FieldNotification|BroadcastFieldValueChanged self "(FieldName=\"EquippedMods\")")))

(fn EquippedMods
  (bind _cacheequippedmods (Variables|Default|GetCacheEquippedMods))
  (bind (_weapondata_itemdefinition_64_e664a9f348977afdcea70ba6c5829726 _weapondata_weaponlevelprogress_41_20c68c5e4244de8471b91bba848297e2 _weapondata_stat0_43_fc7115904a1bf033cc36d0a82fd3a8dc _weapondata_stat1_47_0a2f03a74b3f5506fa4de2a70ef45d5e _weapondata_stat2_48_b2178bd947257b986a0a74ad3b8c2fa1 _weapondata_stat3_49_2b1036a945a55bd6ec75ddb2e2f44d97 _weapondata_modslots_73_5397b13c42ca59a5b3c99aa71d2e1c9c) (Utilities|Struct|BreakSArmTaleWeaponDataFAKE (Variables|GetWeaponData_ItemDefinition_64_E664A9F348977AFDCEA70BA6C5829726)))
  ; Construct equipped mods array for weapon, creating new entries as needed
  (for (_array_element _array_index) _weapondata_modslots_73_5397b13c42ca59a5b3c99aa71d2e1c9c
    (LocalVariables|SetModType _array_element)
    (if (Utilities|Array|IsValidIndex _cacheequippedmods _array_index)
      (LocalVariables|SetModEntry (Utilities|Array|Get(acopy) _cacheequippedmods _array_index))
      (else
        (bind _returnvalue (Game|ConstructObjectfromClass "/ArmTaleFrontendUI/Data/FakeData/VM_ArmTale_ModEntry_FAKE.VM_ArmTale_ModEntry_FAKE_C"))
        (bind _output_get (LocalVariables|SetModEntry _returnvalue))
        (Variables|/ArmTaleFrontendUI/Data/FakeData/VM_ArmTale_ModEntry_FAKE.VM_ArmTale_ModEntry_FAKE_C|SetModData_ItemDefinition_34_D7C19A1B47268B3103B0918B3510F954 (Utilities|Struct|MakeSArmTaleModDataFAKE 0 (LocalVariables|GetModType)) _output_get)))
    (Utilities|Array|Add (LocalVariables|GetOutEquippedMods) (LocalVariables|GetModEntry)))
  (bind _output_get_2 (Variables|Default|SetCacheEquippedMods (LocalVariables|GetOutEquippedMods)))
  (return _output_get_2))

(fn GetEquippedModForSlot
  (params (ModSlotIndex :type Integer))
  (return (Utilities|Array|Get(acopy) (Variables|Default|GetCacheEquippedMods) ModSlotIndex)))
